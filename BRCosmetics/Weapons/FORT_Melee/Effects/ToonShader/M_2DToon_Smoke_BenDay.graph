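# Unlit translucent Niagara material for toon smoke with Ben-Day halftone dots, stored with FORT melee weapon effects.
# A dynamic Time parameter drives a 3x3 flipbook whose masked channel, times particle color and alpha, shapes color and opacity.
# Procedural dots use camera-facing virtual-plane coordinates anchored near the particle, with density adjusted by camera distance.
# Dot color blends with smoke color into emissive adjustments; depth fade softens opacity on non-mobile deferred paths.
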Material /BRCosmetics/Weapons/FORT_Melee/Effects/ToonShader/M_2DToon_Smoke_BenDay
graph {
 properties {
  bUsedWithNiagaraMeshParticles=true
  bUsedWithNiagaraSprites=true
  blendMode=BLEND_Translucent
  pixelDepthOffsetMode=PDOM_Legacy
  shadingModel=MSM_Unlit
  usedShadingModels=Unlit
 }
 nodes [
  n1 = VectorParameter() {parameterName=DotsColor, defaultValue=<rgba 1 0 0.0367 1>}
  n2 = NamedRerouteUsage()
  n3 = ParticleColor()
  n4 = ChannelMaskParameter(None=MaterialFunctionCall(
    "Animation  Phase (0-1)"=DynamicParameter() {paramNames [Time,Picker,Param3,Param4]}.Time
    Number of Rows=Constant() {r=3}
    Number of Columns=Constant() {r=3}
    Texture=TextureObjectParameter() {parameterName=Flipbook Tex, texture=/Game/Weapons/FORT_Melee/Effects/ToonShader/T_2DToon_Flipbook_Smoke}
   ) {materialFunction=&FlipBook}.Result) {maskChannel=Green, defaultValue=<rgba 0 1 0 0>}
  n5 = Multiply(A=@n3.RGB, B=@n4.R)
  n6 = ScalarParameter() {defaultValue=0.5, parameterName=ColorStrenght}
  n7 = Multiply(A=OneMinus(None=@n2), B=@n4.R)
  n8 = ScalarParameter() {defaultValue=0.5, parameterName=BenDayDotsStrenght}
  MaterialFunctionCall(Emissive=LinearInterpolate(A=Multiply(A=@n1.RGB, B=@n2), B=Multiply(A=@n5, B=@n6), Alpha=LinearInterpolate(B=@n7, Alpha=@n8) {constA=1, constAlpha=1})) {materialFunction=&MF_FX_EmissiveAdjustments}
  n9 = ScalarParameter() {defaultValue=20, group=1_RoseDust, parameterName=DotsAmount, sliderMax=512, sortPriority=20}
  n10 = Multiply(A=NamedRerouteUsage(), B=LinearInterpolate(A=@n9, B=Multiply(A=@n9) {constB=0.5}, Alpha=NamedRerouteUsage()))
  n11 = Constant() {r=1}
  n12 = Add(A=Frac(None=Rotator(Coordinate=@n10, Time=@n11) {speed=0.75})) {constB=-0.5}
  n13 = Constant()
  NamedRerouteDeclaration(None=Step(Y=Distance(A=@n12, B=@n13), X=ScalarParameter() {defaultValue=0.25, group=BenDayDots, parameterName=DotSize, sliderMax=0.5, sliderMin=0.01})) {name=BenDayDots, nodeColor=<rgba 0 0.625 1 1>}
  n14 = MaterialFunctionCall() {materialFunction=&CameraDirectionVector}
  n15 = Multiply(A=@n14.Result) {constB=50}
  n16 = ParticlePositionWS()
  n17 = Add(A=@n16, B=Constant3Vector() {constant=<rgba 0 0 90 1>})
  n18 = Constant4Vector() {constant=<rgba 64 128 45 90>}
  n19 = InverseLinearInterpolate(
   A=@n18.B
   B=@n18.B
   Value=MaterialFunctionCall() {materialFunction=/Engine/Functions/Engine_MaterialFunctions02/Utility/FOV}.FOV angle in degrees
  )
  NamedRerouteDeclaration(None=MaterialFunctionCall(
    Plane Normal Axis=Normalize(VectorInput=@n14.Result)
    Plane Center=Add(A=@n15, B=@n17)
    UV Scale=LinearInterpolate(A=@n18.R, B=@n18.R, Alpha=@n19)
   ) {materialFunction=&VirtualPlaneCoordinates}.UVs) {name=ObjectAlighedScreenPosition, nodeColor=<rgba 1 0.867 0 1>}
  n20 = Subtract(A=Distance(A=@n16, B=CameraPositionWS())) {constB=100}
  NamedRerouteDeclaration(None=Saturate(None=Divide(A=@n20) {constB=400})) {name=DistanceBlending, nodeColor=<rgba 0.4766 1 0 1>}
  n21 = Multiply(A=@n3.A, B=@n4.R)
  n22 = DepthFade(Opacity=@n21, FadeDistance=ScalarParameter() {defaultValue=12, parameterName=DepthFadeDis})
  ShadingPathSwitch(Default=@n22, Deferred=@n22, Forward=@n21, Mobile=@n21)
 ]
}
